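# Behavior tree for Phoebe bot looting in Skyfire, driven by AIEvaluator loot keys on the BB_Phoebe blackboard.
# A looping selector gated on the loot destination interacts with the target when the loot interaction execution status reaches a threshold.
# Otherwise it moves to the destination without strafing or wobble, using a no-smashable nav filter.
# The root selector falls back to a short wait when the loot branch fails.
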BehaviorTree /Skyfire/Phoebe/BT_Phoebe_Loot_Skyfire
blackboard=&BB_Phoebe
root {
 name=Selector
 class=BTComposite_Selector
 children [
  Loot Selector {
   class=BTComposite_Selector
   children [
    Loot Interaction {
     class=FortAthenaBTTask_Interact
     children []
     decorators [Can Interact {
       class=BTDecorator_Blackboard
       properties {
        arithmeticOperation=GreaterOrEqual
        intValue=3
        nodeName=Can Interact
        stringValue=""
        blackboardKey {selectedKeyName=AIEvaluator_Loot_InteractionExecutionStatus}
       }
      }]
     properties {
      attemptInterval=0.25
      bIgnoreRestartSelf=false
      bShouldFocusOnInteraction=true
      focalPointName=AIEvaluator_Global_FocalPoint
      interactActorName=AIEvaluator_Global_InteractActor
      jumpExecutionStatusName=AIEvaluator_Jump_ExecutionStatus
      maxInteractAttempts=5
      nodeName=Loot Interaction
      weaponTriggerMeleeName=AIEvaluator_WeaponTriggerMelee_ExecutionStatus
      interactContextInfoKeySelector {bNoneIsAllowedValue=false, selectedKeyName=AIEvaluator_Loot_InteractionContextInfo}
      interactExecutionStatusKeySelector {bNoneIsAllowedValue=false, selectedKeyName=AIEvaluator_Loot_InteractionExecutionStatus}
     }
     services []
    }
    Loot Move To {
     class=FortAthenaBTTask_BotMoveTo
     decorators []
     properties {
      movementResultKeyName=AIEvaluator_Loot_MovementState
      nodeName=Loot Move To
      acceptableRadius {defaultValue=50}
      allowRandomWobble {defaultValue=false}
      bAllowStrafe {defaultValue=false}
      blackboardKey {selectedKeyName=AIEvaluator_Loot_Destination}
      filterClass {defaultValue=&NavFilter_Phoebe_NoSmashable_C}
      observedBlackboardValueTolerance {defaultValue=4.75}
     }
    }
   ]
   decorators [
    Has Valid Loot Dest {
     class=BTDecorator_Blackboard
     properties {
      arithmeticOperation=Equal
      intValue=0
      nodeName=Has Valid Loot Dest
      stringValue=""
      blackboardKey {selectedKeyName=AIEvaluator_Loot_Destination}
     }
    }
    Loop when blocked {class=BTDecorator_Loop, properties {nodeName=Loop when blocked}}
   ]
   properties {
    nodeName=Loot Selector
    children [
     {
      childComposite=null
      childTask=&self:FortAthenaBTTask_Interact_0
      decorators [&self:BTDecorator_Blackboard_26]
     }
     {childComposite=null, childTask=&self:FortAthenaBTTask_BotMoveTo_9, decorators []}
    ]
    services []
   }
   services []
  }
  Wait {class=BTTask_Wait, properties {waitTime {defaultValue=0.7}}}
 ]
 decorators []
 properties {
  children [
   {
    childComposite=&self:BTComposite_Selector_23
    childTask=null
    decorators [&self:BTDecorator_Blackboard_0, &self:BTDecorator_Loop_0]
   }
   {childComposite=null, childTask=&self:BTTask_Wait_0, decorators []}
  ]
  services []
 }
 services []
}
