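# Behavior tree for Phoebe AI bots in Skyfire that moves a bot to its current patrol destination.
# A selector first runs a bot move-to on the evaluator's patrol destination key, reporting to a movement-state key, with strafing allowed and a no-smashable nav filter.
# If the move fails, a brief wait task runs as a fallback so the selector still returns success.
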BehaviorTree /Skyfire/Phoebe/BT_Phoebe_PatrolAround_Skyfire
blackboard=&BB_Phoebe
root {
 name=Patrol Move Move Selector
 class=BTComposite_Selector
 children [
  Patrol Move To {
   class=FortAthenaBTTask_BotMoveTo
   decorators []
   properties {
    movementResultKeyName=AIEvaluator_Patrol_MovementState
    nodeName=Patrol Move To
    acceptableRadius {defaultValue=5}
    allowRandomWobble {defaultValue=false}
    bAllowStrafe {defaultValue=true}
    bTrackMovingGoal {defaultValue=false}
    blackboardKey {selectedKeyName=AIEvaluator_Patrol_Destination}
    filterClass {defaultValue=&NavFilter_Phoebe_NoSmashable_C}
    observedBlackboardValueTolerance {defaultValue=200}
   }
  }
  Wait so above selector return success {class=BTTask_Wait, properties {nodeName=Wait so above selector return success, waitTime {defaultValue=0.1}}}
 ]
 decorators []
 properties {
  nodeName=Patrol Move Move Selector
  children [
   {childComposite=null, childTask=&self:FortAthenaBTTask_BotMoveTo_1, decorators []}
   {childComposite=null, childTask=&self:BTTask_Wait_0, decorators []}
  ]
  services []
 }
 services []
}
