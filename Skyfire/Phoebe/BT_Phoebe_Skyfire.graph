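# Behavior tree for Phoebe AI bots in the Skyfire mode, prioritizing prison, gameplay and loot room phases from blackboard keys, with a long idle wait as fallback.
# Prison and blocked movement loop on waits; threatened bots steer when launched or patrol under evasion, focus and ranged-attack evaluators, while others patrol and scan.
# The loot room runs a loot subtree, falling back to patrol.
# Evaluator services write the keys gating branches, alongside weapon, jump, sprint and crouch services.
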BehaviorTree /Skyfire/Phoebe/BT_Phoebe_Skyfire
blackboard=&/BB_Phoebe_Skyfire
root {
 name=Selector
 class=BTComposite_Selector
 children [
  Wait {
   class=BTTask_Wait
   decorators [
    In Prison? {
     class=BTDecorator_Blackboard
     properties {arithmeticOperation=Equal, intValue=0, nodeName=In Prison?, stringValue="", blackboardKey {selectedKeyName=AIEvaluator_Skyfire_InPrison}}
    }
    Loop {class=BTDecorator_Loop}
   ]
   properties {waitTime {defaultValue=1}}
  }
  Gameplay {
   class=BTComposite_Selector
   children [
    Wait {
     class=BTTask_Wait
     decorators [
      Is Movement Blocked? {
       class=BTDecorator_Blackboard
       properties {
        arithmeticOperation=Equal
        intValue=0
        nodeName=Is Movement Blocked?
        stringValue=""
        blackboardKey {selectedKeyName=AIEvaluator_Global_IsMovementBlocked}
       }
      }
      Loop {class=BTDecorator_Loop}
     ]
     properties {waitTime {defaultValue=1}}
    }
    On Ground {
     class=BTComposite_Selector
     children [
      THREATEN {
       class=BTComposite_Selector
       children [
        Steer Movement {
         class=FortAthenaBTTask_SteerMovement
         children []
         decorators [Is Character being launched? {
           class=BTDecorator_Blackboard
           properties {
            arithmeticOperation=GreaterOrEqual
            intValue=3
            nodeName=Is Character being launched?
            stringValue=EExecutionStatus::ExecutionPending
            blackboardKey {selectedKeyName=AIEvaluator_CharacterLaunched_ExecutionStatus}
           }
          }]
         properties {
          bIgnoreRestartSelf=false
          bSetControlRotation=false
          nodeName=Steer Movement
          steerDirectionKeySelector {bNoneIsAllowedValue=false, selectedKeyName=AIEvaluator_CharacterLaunched_SteerDirection}
         }
         services []
        }
        Run PatrolAround Behavior {
         class=BTTask_RunBehavior
         decorators [Patrol Around {
           class=BTDecorator_Blackboard
           properties {
            arithmeticOperation=GreaterOrEqual
            intValue=2
            nodeName=Patrol Around
            stringValue=EExecutionStatus::ExecutionSuccess
            blackboardKey {selectedKeyName=AIEvaluator_Patrol_ExecutionStatus}
           }
          }]
         properties {behaviorAsset=&/BT_Phoebe_PatrolAround_Skyfire, nodeName=Run PatrolAround Behavior}
        }
        Wait {class=BTTask_Wait, properties {waitTime {defaultValue=2}}}
       ]
       decorators [IsThreatened {
         class=BTDecorator_Blackboard
         properties {
          arithmeticOperation=Equal
          intValue=3
          nodeName=IsThreatened
          stringValue=EAlertLevel::Threatened
          blackboardKey {selectedKeyName=AIEvaluator_Global_AlertLevel}
         }
        }]
       properties {
        nodeName=THREATEN
        children [
         {
          childComposite=null
          childTask=&self:FortAthenaBTTask_SteerMovement_0
          decorators [&self:BTDecorator_Blackboard_0]
         }
         {
          childComposite=null
          childTask=&self:BTTask_RunBehavior_2
          decorators [&self:BTDecorator_Blackboard_24]
         }
         {childComposite=null, childTask=&self:BTTask_Wait_31, decorators []}
        ]
        services [
         &self:FortAthenaBTService_AIEvaluator_4
         &self:FortAthenaBTService_AIEvaluator_5
         &self:FortAthenaBTService_AIEvaluator_7
         &self:FortAthenaBTService_AIEvaluator_8
         &self:FortAthenaBTService_ManageWeapon_0
        ]
       }
       services [
        Evaluating... Evasive Maneuvers {
         class=FortAthenaBTService_AIEvaluator
         properties {
          interval=0.5
          nodeName=Evaluating... Evasive Maneuvers
          aIEvaluatorClass {defaultValue=&/Phoebe_EvasiveManeuvers_Skyfire.Phoebe_EvasiveManeuvers_Skyfire_C}
         }
        }
        Focus on target {
         class=FortAthenaBTService_AIEvaluator
         properties {interval=0.5, nodeName=Focus on target, aIEvaluatorClass {defaultValue=&FortAthenaAIBotEvaluator_HandleFocusing}}
        }
        Evaluating ... Patrol Around {
         class=FortAthenaBTService_AIEvaluator
         properties {
          bCallTickOnSearchStart=true
          interval=2.5
          nodeName=Evaluating ... Patrol Around
          randomDeviation=0.5
          aIEvaluatorClass {defaultValue=/Script/FortniteAIServer.FortAthenaAIBotEvaluator_PatrolAround}
         }
        }
        Evaluating...Range Attack {
         class=FortAthenaBTService_AIEvaluator
         properties {
          bCallTickOnSearchStart=true
          interval=0.2
          nodeName=Evaluating...Range Attack
          randomDeviation=0.05
          aIEvaluatorClass {defaultValue=&FortAthenaAIBotEvaluator_RangeAttack}
         }
        }
        Executing...Weapons Actions {class=FortAthenaBTService_ManageWeapon, properties {interval=1, randomDeviation=0.3}}
       ]
      }
      Not attacking threat {
       class=BTComposite_Selector
       children [
        Steer Movement {
         class=FortAthenaBTTask_SteerMovement
         children []
         decorators [Is Character being launched? {
           class=BTDecorator_Blackboard
           properties {
            arithmeticOperation=GreaterOrEqual
            intValue=3
            nodeName=Is Character being launched?
            stringValue=EExecutionStatus::ExecutionPending
            blackboardKey {selectedKeyName=AIEvaluator_CharacterLaunched_ExecutionStatus}
           }
          }]
         properties {
          bIgnoreRestartSelf=false
          bSetControlRotation=false
          nodeName=Steer Movement
          steerDirectionKeySelector {bNoneIsAllowedValue=false, selectedKeyName=AIEvaluator_CharacterLaunched_SteerDirection}
         }
         services []
        }
        UNAWARE {
         class=BTComposite_Selector
         children [
          Run PatrolAround Behavior {
           class=BTTask_RunBehavior
           decorators [Patrol Around {
             class=BTDecorator_Blackboard
             properties {
              arithmeticOperation=GreaterOrEqual
              intValue=2
              nodeName=Patrol Around
              stringValue=EExecutionStatus::ExecutionSuccess
              blackboardKey {selectedKeyName=AIEvaluator_Patrol_ExecutionStatus}
             }
            }]
           properties {behaviorAsset=&BT_Phoebe_PatrolAround_Skyfire, nodeName=Run PatrolAround Behavior}
          }
          Wait {class=BTTask_Wait, properties {waitTime {defaultValue=2}}}
         ]
         decorators [Loop {class=BTDecorator_Loop}]
         properties {
          nodeName=UNAWARE
          children [
           {
            childComposite=null
            childTask=&self:BTTask_RunBehavior_12
            decorators [&self:BTDecorator_Blackboard_25]
           }
           {childComposite=null, childTask=&self:BTTask_Wait_5, decorators []}
          ]
          services [&self:FortAthenaBTService_AIEvaluator_6, &self:FortAthenaBTService_AIEvaluator_39]
         }
         services [
          Evaluating ... Patrol Around {
           class=FortAthenaBTService_AIEvaluator
           properties {
            bCallTickOnSearchStart=true
            interval=2.5
            nodeName=Evaluating ... Patrol Around
            randomDeviation=0.5
            aIEvaluatorClass {defaultValue=&FortAthenaAIBotEvaluator_PatrolAround}
           }
          }
          Focus Scan Around {
           class=FortAthenaBTService_AIEvaluator
           properties {
            interval=0.5
            nodeName=Focus Scan Around
            aIEvaluatorClass {defaultValue=/Game/Athena/AI/Phoebe/Evaluators/Phoebe_HandleFocusing_ScanAround.Phoebe_HandleFocusing_ScanAround_C}
           }
          }
         ]
        }
       ]
       decorators []
       properties {
        nodeName=Not attacking threat
        children [
         {
          childComposite=null
          childTask=&self:FortAthenaBTTask_SteerMovement_2
          decorators [&self:BTDecorator_Blackboard_1]
         }
         {
          childComposite=&self:BTComposite_Selector_16
          childTask=null
          decorators [&self:BTDecorator_Loop_2]
         }
        ]
        services []
       }
       services []
      }
     ]
     decorators []
     properties {
      nodeName=On Ground
      children [
       {
        childComposite=&self:BTComposite_Selector_33
        childTask=null
        decorators [&self:BTDecorator_Blackboard_42]
       }
       {childComposite=&self:BTComposite_Selector_31, childTask=null, decorators []}
      ]
      services [&self:FortAthenaBTService_Crouch_0, &self:FortAthenaBTService_AIEvaluator_61]
     }
     services [
      Executing...Crouch {class=FortAthenaBTService_Crouch}
      Evaluating...Ground {
       class=FortAthenaBTService_AIEvaluator
       properties {interval=1, nodeName=Evaluating...Ground, aIEvaluatorClass {defaultValue=/Script/FortniteAIServer.FortAthenaAIBotEvaluator_Ground}}
      }
     ]
    }
   ]
   decorators [In Gameplay? {
     class=BTDecorator_Blackboard
     properties {
      arithmeticOperation=GreaterOrEqual
      intValue=3
      nodeName=In Gameplay?
      stringValue=""
      blackboardKey {selectedKeyName=AIEvaluator_Skyfire_InGameplay}
     }
    }]
   properties {
    nodeName=Gameplay
    children [
     {
      childComposite=null
      childTask=&self:BTTask_Wait_7
      decorators [&self:BTDecorator_Blackboard_16, &self:BTDecorator_Loop_4]
     }
     {childComposite=&self:BTComposite_Selector_17, childTask=null, decorators []}
    ]
    services [
     &self:FortAthenaBTService_AIEvaluator_11
     &self:FortAthenaBTService_UpdateTarget_0
     &self:FortAthenaBTService_AIEvaluator_65
     &self:FortAthenaBTService_AIEvaluator_62
     &self:FortAthenaBTService_AIEvaluator_3
     &self:FortAthenaBTService_ManageWeapon_5
     &self:FortAthenaBTService_Jump_0
     &self:FortAthenaBTService_Sprinting_0
    ]
   }
   services [
    Evaluating...Select Weapon {
     class=FortAthenaBTService_AIEvaluator
     properties {
      interval=0.2
      nodeName=Evaluating...Select Weapon
      randomDeviation=0.05
      aIEvaluatorClass {defaultValue=&FortAthenaAIBotEvaluator_WeaponSelection}
     }
    }
    Evaluating...Targets {class=FortAthenaBTService_UpdateTarget, properties {interval=0.1}}
    Evaluating...Character Launched {
     class=FortAthenaBTService_AIEvaluator
     properties {
      interval=0.5
      nodeName=Evaluating...Character Launched
      aIEvaluatorClass {defaultValue=&/Phoebe_Evaluator_CharacterLaunched_Skyfire.Phoebe_Evaluator_CharacterLaunched_Skyfire_C}
     }
    }
    Evaluate...Sprinting {
     class=FortAthenaBTService_AIEvaluator
     properties {interval=0.2, nodeName=Evaluate...Sprinting, aIEvaluatorClass {defaultValue=/Script/FortniteAIServer.FortAthenaAIBotEvaluator_Sprinting}}
    }
    Evaluating...Range Attack {
     class=FortAthenaBTService_AIEvaluator
     properties {
      bCallTickOnSearchStart=true
      interval=0.2
      nodeName=Evaluating...Range Attack
      randomDeviation=0.05
      aIEvaluatorClass {defaultValue=&FortAthenaAIBotEvaluator_RangeAttack}
     }
    }
    Executing...Weapons Actions {class=FortAthenaBTService_ManageWeapon, properties {interval=1, randomDeviation=0.3}}
    Executing...Jump {class=FortAthenaBTService_Jump, properties {crouchExecutionStatusName=AIEvaluator_Crouch_ExecutionStatus}}
    Executing...Sprinting {class=FortAthenaBTService_Sprinting}
   ]
  }
  Selector {
   class=BTComposite_Selector
   children [
    Run Loot Behavior {
     class=BTTask_RunBehavior
     decorators [Loot {
       class=BTDecorator_Blackboard
       properties {
        arithmeticOperation=GreaterOrEqual
        intValue=2
        nodeName=Loot
        stringValue=EExecutionStatus::ExecutionSuccess
        blackboardKey {selectedKeyName=AIEvaluator_Loot_ExecutionStatus}
       }
      }]
     properties {behaviorAsset=&/BT_Phoebe_Loot_Skyfire, nodeName=Run Loot Behavior}
    }
    Fallback Behavior {
     class=BTComposite_Selector
     children [
      Run PatrolAround Behavior {
       class=BTTask_RunBehavior
       decorators [Patrol Around {
         class=BTDecorator_Blackboard
         properties {
          arithmeticOperation=GreaterOrEqual
          intValue=2
          nodeName=Patrol Around
          stringValue=EExecutionStatus::ExecutionSuccess
          blackboardKey {selectedKeyName=AIEvaluator_Patrol_ExecutionStatus}
         }
        }]
       properties {behaviorAsset=&BT_Phoebe_PatrolAround_Skyfire, nodeName=Run PatrolAround Behavior}
      }
      Wait {class=BTTask_Wait, properties {waitTime {defaultValue=2}}}
     ]
     decorators []
     properties {
      nodeName=Fallback Behavior
      children [
       {
        childComposite=null
        childTask=&self:BTTask_RunBehavior_3
        decorators [&self:BTDecorator_Blackboard_30]
       }
       {childComposite=null, childTask=&self:BTTask_Wait_0, decorators []}
      ]
      services [&self:FortAthenaBTService_AIEvaluator_23]
     }
     services [Evaluating ... Patrol Around {
       class=FortAthenaBTService_AIEvaluator
       properties {
        bCallTickOnSearchStart=true
        interval=2.5
        nodeName=Evaluating ... Patrol Around
        randomDeviation=0.5
        aIEvaluatorClass {defaultValue=&FortAthenaAIBotEvaluator_PatrolAround}
       }
      }]
    }
   ]
   decorators [In Loot Room? {
     class=BTDecorator_Blackboard
     properties {
      arithmeticOperation=Equal
      intValue=0
      nodeName=In Loot Room?
      stringValue=""
      blackboardKey {selectedKeyName=AIEvaluator_Skyfire_InLootRoom}
     }
    }]
   properties {
    children [
     {
      childComposite=null
      childTask=&self:BTTask_RunBehavior_0
      decorators [&self:BTDecorator_Blackboard_7]
     }
     {childComposite=&self:BTComposite_Selector_18, childTask=null, decorators []}
    ]
    services [
     &self:FortAthenaBTService_AIEvaluator_17
     &self:FortAthenaBTService_AIEvaluator_18
     &self:FortAthenaBTService_PickUpLoot_1
     &self:FortAthenaBTService_Jump_3
     &self:FortAthenaBTService_Sprinting_3
    ]
   }
   services [
    Evaluate...Loot {
     class=FortAthenaBTService_AIEvaluator
     properties {
      bCallTickOnSearchStart=true
      interval=0.5
      nodeName=Evaluate...Loot
      aIEvaluatorClass {defaultValue=/Script/FortniteAI.FortAthenaAIBotEvaluator_Loot}
     }
    }
    Evaluate...Sprinting {
     class=FortAthenaBTService_AIEvaluator
     properties {interval=0.2, nodeName=Evaluate...Sprinting, aIEvaluatorClass {defaultValue=&FortAthenaAIBotEvaluator_Sprinting}}
    }
    Executing...Pick Up Loot {
     class=FortAthenaBTService_PickUpLoot
     properties {
      bCallTickOnSearchStart=false
      bRestartTimerOnEachActivation=false
      executionStatusName=AIEvaluator_Loot_ExecutionStatus
      interactionContextInfoName=AIEvaluator_Loot_InteractionContextInfo
      interactionExecutionStatusName=AIEvaluator_Loot_InteractionExecutionStatus
      interval=0.5
      lootObjectKeyName=AIEvaluator_Loot_Object
      movementStateKeyName=AIEvaluator_Loot_MovementState
      nodeName=Executing...Pick Up Loot
      randomDeviation=0.1
     }
    }
    Executing...Jump {class=FortAthenaBTService_Jump, properties {crouchExecutionStatusName=AIEvaluator_Crouch_ExecutionStatus}}
    Executing...Sprinting {class=FortAthenaBTService_Sprinting}
   ]
  }
  Wait {class=BTTask_Wait, properties {waitTime {defaultValue=999}}}
 ]
 decorators []
 properties {
  children [
   {
    childComposite=null
    childTask=&self:BTTask_Wait_6
    decorators [&self:BTDecorator_Blackboard_18, &self:BTDecorator_Loop_6]
   }
   {
    childComposite=&self:BTComposite_Selector_7
    childTask=null
    decorators [&self:BTDecorator_Blackboard_11]
   }
   {
    childComposite=&self:BTComposite_Selector_1
    childTask=null
    decorators [&self:BTDecorator_Blackboard_3]
   }
   {childComposite=null, childTask=&self:BTTask_Wait_8, decorators []}
  ]
  services []
 }
 services []
}
